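# Niagara system for the MarineCarve emote that adds a red glow and a blue shine to sockets on the gadget prop's camera.
# Minimal pins a two-renderer sprite glow to JNT_CameraSocket, fading it in with emitter age after a short delay.
# Camera_Shine repeatedly bursts a probability-gated mesh on JNT_CameraSocket_Sheen, animated through a curve-driven dynamic material parameter.
# Both emitters use camera offset and scale opacity by time of day.
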NiagaraSystem /BRCosmetics/Animation/Game/MainPlayer/Emotes/MarineCarve/FX/NS_Emote_MarineCarve_Glow
dependencies {
 usedDataInterfaces [&NiagaraDataInterfaceSocketReader, &NiagaraDataInterfaceCurve]
 usedDynamicInputs []
 usedModules [&SystemState]
}
emitters {
 Camera_Shine {
  emitterUpdateScript [
   &EmitterState(Life Cycle Mode=::NewEnumerator1, Loop Delay=1, Loop Duration=3, Recalculate Duration Each Loop=true, UseLoopDelay=true)
   &SpawnBurst_Instantaneous(Spawn Probability=0.7, Use Spawn Probability=true, UseLoopCountLimit=false)
  ]
  particleSpawnScript [
   &InitializeParticle(
    Color=<rgba 0.3524 0.7081 1 0.15>
    Lifetime=2
    Mesh Scale=&Multiply_VectorByFloat(Vector=<1 1 1.2>, Float=0.035)
    Mesh Scale Mode=::NewEnumerator4
    UsePositionOffset=true
    Write Material Random=true
   )
   &CameraOffset(Camera Offset Amount=1)
   &InitialMeshOrientation(
    Mesh Axis To Orient=<0 0 1>
    Mesh Orientation Mode=::NewEnumerator2
    Orientation Vector=<0 1 0>
    Rotation=<0 0.25 0>
    Use Rotation Vector=true
   )
  ]
  particleUpdateScript [
   &ParticleState()
   &SocketLocation(
    SamplePosOffset=false
    Socket Reader=DataInterface() {propertyValues {
      AttachComponentClass=null
      AttachComponentTag=None
      EditorPreviewAsset=/BRCosmetics/Animation/Props/Gadgets_MarineCarve/Meshes/MarineCarve_Gadget
      SourceActorRef=null
      SourceAsset=null
      SourceMode=Default
      bRequireCurrentFrameData=true
      bUpdateSocketsPerFrame=true
      FilteredSockets [JNT_CameraSocket_Sheen]
      ObjectParameterBinding {parameter {name=None, typeDefHandle {registeredTypeIndex=1}, varData []}}
     }}
    Socket Sampling Mode=::NewEnumerator3
    VelocitySpeedLimit=false
   )
   &DynamicMaterialParameters(
    Index 0 Param 1=&FloatFromCurve(FloatCurve=DataInterface() {propertyValues {Curve {keys [{interpMode=RCIM_Cubic, tangentMode=RCTM_Break, time=0, value=0}, {arriveTangent=2.3933, tangentMode=RCTM_User, time=1, value=1}]}}})
    Index 0 Param 2=1
    Index 0 Param 3=-0.1138
    Module.Index0Float0Write=true
    Module.Index0Float1Write=true
    Module.Index0Float2Write=true
    Module.Index0Float3Write=true
   )
   &SolveForcesAndVelocity(Clamp Velocity=false, Limit Acceleration=false)
   &ScaleColor(
    Scale Alpha=&FNTimeOfDayScalar(Night Scale=0.4, Morning Scale=0.5, Day Scale=0.6, Evening Scale=0.5)
    ScaleA=true
    ScaleRGB=false
   )
  ]
  rendererClasses [&NiagaraMeshRendererProperties]
  renderers [{rendererClass=&NiagaraMeshRendererProperties, rendererIndex=0}]
 }
 Minimal {
  emitterUpdateScript [
   &EmitterState(Life Cycle Mode=::NewEnumerator1, Loop Behavior=::NewEnumerator1, Loop Duration Mode=::NewEnumerator1, UseLoopDelay=false)
   &SpawnBurst_Instantaneous(Use Spawn Probability=false, UseLoopCountLimit=false)
  ]
  particleSpawnScript [
   &InitializeParticle(
    Color=<rgba 1 0.1146 0.1146 1>
    Lifetime=5
    Sprite Size Mode=::NewEnumerator3
    Uniform Sprite Size=15
    UsePositionOffset=false
    Write Material Random=true
   )
   /SetVariables(Particles.OuterGlowSize:Vector2f=<30 30>) {isSetParameters=true}
   &CameraOffset(Camera Offset Amount=5)
  ]
  particleUpdateScript [
   &ParticleState(Kill Particles When Lifetime Has Elapsed=false, Let Infinitely Lived Particles Die When Emitter Deactivates=true)
   &SocketLocation(
    SamplePosOffset=true
    Sampled Position Offset Coordinate Space:ENiagaraSkelMeshTransforms=::NewEnumerator2
    Socket Reader=DataInterface() {propertyValues {
      AttachComponentClass=null
      AttachComponentTag=None
      EditorPreviewAsset=&MarineCarve_Gadget
      SourceActorRef=null
      SourceAsset=null
      SourceMode=Default
      bRequireCurrentFrameData=true
      bUpdateSocketsPerFrame=true
      FilteredSockets [JNT_CameraSocket]
      ObjectParameterBinding {parameter {name=None, typeDefHandle {registeredTypeIndex=1}, varData []}}
     }}
    Socket Sampling Mode=::NewEnumerator3
    VelocitySpeedLimit=false
   )
   &ScaleColor(
    Scale Alpha=&FloatFromCurve(
     FloatCurve=DataInterface() {propertyValues {Curve {
        defaultValue=0
        keys [
         {arriveTangent=4.1785, interpMode=RCIM_Cubic, leaveTangent=4.1785, tangentMode=RCTM_User, time=0.484, value=0}
         {arriveTangent=1.1179, interpMode=RCIM_Cubic, leaveTangent=1.1179, tangentMode=RCTM_User, time=0.6258, value=0.6228}
         {arriveTangent=0.0224, interpMode=RCIM_Cubic, leaveTangent=0.0224, tangentMode=RCTM_User, time=1, value=1}
        ]
       }}}
     CurveIndex=$Emitter.Age
     Scale Curve=&FNTimeOfDayScalar(Night Scale=0.15, Morning Scale=0.2, Day Scale=0.3, Evening Scale=0.15)
    )
    ScaleA=true
    ScaleRGB=true
   )
   /SetVariables(Particles.HotCircleColor:LinearColor=&MakeLinearColorFromVectorAndFloat(
     "Vector (RGB)"=&MakeVectorFromLinearColorRGB(Color=<rgba 0.5365 0.0754 0.0754 1>)
     "Float (Alpha)"=&FloatFromCurve(
      FloatCurve=DataInterface() {propertyValues {Curve {
         defaultValue=0
         keys [
          {arriveTangent=4.1785, interpMode=RCIM_Cubic, leaveTangent=4.1785, tangentMode=RCTM_User, time=0.484, value=0}
          {arriveTangent=1.1179, interpMode=RCIM_Cubic, leaveTangent=1.1179, tangentMode=RCTM_User, time=0.6258, value=0.6228}
          {arriveTangent=0.0224, interpMode=RCIM_Cubic, leaveTangent=0.0224, tangentMode=RCTM_User, time=1, value=1}
         ]
        }}}
      CurveIndex=$Emitter.Age
      Scale Curve=&FNTimeOfDayScalar(Night Scale=0.15, Morning Scale=0.2, Day Scale=0.3, Evening Scale=0.15)
     )
    )) {isSetParameters=true}
  ]
  rendererClasses [&NiagaraSpriteRendererProperties]
  renderers [
   {rendererClass=&NiagaraSpriteRendererProperties}
   {rendererClass=&NiagaraSpriteRendererProperties, rendererIndex=1}
  ]
 }
}
summary {
 systemName=NS_Emote_MarineCarve_Glow
 emitters [
  {emitterName=Minimal, rendererClasses [&NiagaraSpriteRendererProperties]}
  {emitterName=Camera_Shine, rendererClasses [&NiagaraMeshRendererProperties]}
 ]
 userVariables []
}
